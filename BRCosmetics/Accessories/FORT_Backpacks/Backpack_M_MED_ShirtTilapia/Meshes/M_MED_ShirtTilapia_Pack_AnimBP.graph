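# Animation Blueprint for the ShirtTilapia backpack cosmetic that plays its idle sequence and adds a floating bob to five ball bones.
# Each update, the event graph computes five Z-offset vectors from sine waves on real time, with differing frequencies and amplitudes, some inverted, so the balls move out of sync.
# The AnimGraph applies these as additive translations through chained Modify Bone nodes on the dyn ball bones before output.
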AnimBlueprint /BRCosmetics/Accessories/FORT_Backpacks/Backpack_M_MED_ShirtTilapia/Meshes/M_MED_ShirtTilapia_Pack_AnimBP
target_skeleton=&/M_MED_ShirtTilapia_Pack_Skeleton
graphs [
 AnimGraph {nodes [
   n1 = "Transform (Modify) Bone - Bone: dyn_shirtTilapia_ball_1"(
    ComponentPose=Local To Component(LocalPose=Sequence Player 'ShirtTilapia_Idle_PACK'() {sequence=/BRCosmetics/Accessories/FORT_Backpacks/Backpack_M_MED_ShirtTilapia/Anim/ShirtTilapia_Idle_PACK})
    Alpha=1
    Translation=Get FloatingMotion1() -> (FloatingMotion1:Vector)
    Rotation="0,0,0"
    Scale="1,1,1"
   ) {translationMode=BMM_Additive, alphaBoolBlend {blendOption=Linear}, boneToModify {boneName=dyn_shirtTilapia_ball_1}}
   n2 = Get FloatingMotion2() -> (FloatingMotion2:Vector)
   n3 = "Transform (Modify) Bone - Bone: dyn_shirtTilapia_ball_4"(
    ComponentPose="Transform (Modify) Bone - Bone: dyn_shirtTilapia_ball_3"(
     ComponentPose="Transform (Modify) Bone - Bone: dyn_shirtTilapia_ball_2"(ComponentPose=@n1, Alpha=1, Translation=@n2, Rotation="0,0,0", Scale="1,1,1") {translationMode=BMM_Additive, alphaBoolBlend {blendOption=Linear}, boneToModify {boneName=dyn_shirtTilapia_ball_2}}
     Alpha=1
     Translation=Get FloatingMotion3() -> (FloatingMotion3:Vector)
     Rotation="0,0,0"
     Scale="1,1,1"
    ) {translationMode=BMM_Additive, alphaBoolBlend {blendOption=Linear}, boneToModify {boneName=dyn_shirtTilapia_ball_3}}
    Alpha=1
    Translation=Get FloatingMotion4() -> (FloatingMotion4:Vector)
    Rotation="0,0,0"
    Scale="1,1,1"
   ) {translationMode=BMM_Additive, alphaBoolBlend {blendOption=Linear}, boneToModify {boneName=dyn_shirtTilapia_ball_4}}
   n4 = Get FloatingMotion5() -> (FloatingMotion5:Vector)
   Output Pose(Result=Component To Local(ComponentPose="Transform (Modify) Bone - Bone: dyn_shirtTilapia_ball_5"(ComponentPose=@n3, Alpha=1, Translation=@n4, Rotation="0,0,0", Scale="1,1,1") {translationMode=BMM_Additive, alphaBoolBlend {blendOption=Linear}, boneToModify {boneName=dyn_shirtTilapia_ball_5}})) {layerGroup=None}
  ]}
 EventGraph {nodes [
   Event BlueprintUpdateAnimation()
   > n1 = Sequence!(execute) -> (then_0, then_1, then_2:Exec, then_3:Exec, then_4:Exec)
   n2 = "Sin (Radians)"(A=float * float(A=float * float(A=1.2, B=-1.2), B=GetRealTimeSeconds()) {is=title K2Node_PromotableOperator})
   > Set FloatingMotion1(FloatingMotion1:Vector=MakeVector(Y=0, Z=float / float(A=@n2, B=2))) -> (Output_Get:Vector)
   n3 = "Sin (Radians)"(A=float * float(A=float * float(A=1.1, B=-1.1), B=GetRealTimeSeconds()))
   Set FloatingMotion2(execute=@n1.then_1, FloatingMotion2:Vector=MakeVector(Y=0, Z=float / float(A=@n3, B=2.2))) -> (Output_Get:Vector)
   n4 = "Sin (Radians)"(A=float * float(A=float * float(A=0.95, B=-0.95), B=GetRealTimeSeconds()))
   Set FloatingMotion3(execute=@n1.then_2, FloatingMotion3:Vector=MakeVector(Y=0, Z=float / float(A=@n4, B=-2.3))) -> (Output_Get:Vector)
   n5 = "Sin (Radians)"(A=float * float(A=float * float(A=1.25, B=-1.25), B=GetRealTimeSeconds()))
   Set FloatingMotion5(execute=@n1.then_4, FloatingMotion5:Vector=MakeVector(Y=0, Z=float / float(A=@n5, B=-1.75))) -> (Output_Get:Vector)
   n6 = "Sin (Radians)"(A=float * float(A=float * float(A=1, B=-1), B=GetRealTimeSeconds()))
   Set FloatingMotion4(execute=@n1.then_3, FloatingMotion4:Vector=MakeVector(Y=0, Z=float / float(A=@n6, B=2.25))) -> (Output_Get:Vector)
  ]}
]
variables [
 FloatingMotion1 {type=Vector}
 FloatingMotion2 {type=Vector}
 FloatingMotion3 {type=Vector}
 FloatingMotion4 {type=Vector}
 FloatingMotion5 {type=Vector}
]
